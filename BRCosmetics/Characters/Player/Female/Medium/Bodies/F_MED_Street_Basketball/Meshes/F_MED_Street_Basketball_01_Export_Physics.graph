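# Physics asset for the female medium Street Basketball body mesh, providing physics bodies for its dynamic hair bones.
# Six hair bones, the head and the pelvis each use a single sphere, with no other body parts covered.
# Each hair bone constrains directly to the head rather than forming a chain, and the head connects to the pelvis with free rotation.
# All hair joints lock swing1, and dyn_hair_06 also locks swing2 and twist.
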PhysicsAsset /BRCosmetics/Characters/Player/Female/Medium/Bodies/F_MED_Street_Basketball/Meshes/F_MED_Street_Basketball_01_Export_Physics
bodies {
 dyn_hair_01 [{shapeType=Sphere, radius=6.8544, center=<-0.0174 1.5884 -0.2166>}]
 dyn_hair_02 [{shapeType=Sphere, radius=6.8544, center=<-0.0174 1.5884 -0.2166>}]
 dyn_hair_03 [{shapeType=Sphere, radius=6.8544, center=<-0.0174 -0.6955 -1.7413>}]
 dyn_hair_04 [{shapeType=Sphere, radius=6.8544, center=<-0.0174 -0.7509 -2.0626>}]
 dyn_hair_05 [{shapeType=Sphere, radius=8.0811, center=<-0.054 1.7047 2.2475>}]
 dyn_hair_06 [{shapeType=Sphere, radius=2.6872, center=<-0.0411 -1.7552 0.0806>}]
 head [{shapeType=Sphere, radius=3.0824, center=<-0.933 -2.2779 -0.0001>}]
 pelvis [{shapeType=Sphere, radius=5.2359, center=<2.0487 -0.2833 0>}]
}
constraints [
 {bone1Name=dyn_hair_01, bone2Name=head, swing1LimitDegrees=20, swing1Motion=Locked, swing2LimitDegrees=20, twistLimitDegrees=20}
 {bone1Name=dyn_hair_02, bone2Name=head, swing1LimitDegrees=20, swing1Motion=Locked, swing2LimitDegrees=20, twistLimitDegrees=20}
 {bone1Name=dyn_hair_03, bone2Name=head, swing1LimitDegrees=20, swing1Motion=Locked, swing2LimitDegrees=20, twistLimitDegrees=20}
 {bone1Name=dyn_hair_04, bone2Name=head, swing1LimitDegrees=20, swing1Motion=Locked, swing2LimitDegrees=20, twistLimitDegrees=20}
 {bone1Name=dyn_hair_05, bone2Name=head, swing1LimitDegrees=20, swing1Motion=Locked, swing2LimitDegrees=20, twistLimitDegrees=20}
 {
  bone1Name=dyn_hair_06
  bone2Name=head
  swing1LimitDegrees=10
  swing1Motion=Locked
  swing2LimitDegrees=10
  swing2Motion=Locked
  twistLimitDegrees=0
  twistMotion=Locked
 }
 {
  bone1Name=head
  bone2Name=pelvis
  swing1LimitDegrees=45
  swing1Motion=Free
  swing2LimitDegrees=45
  swing2Motion=Free
  twistLimitDegrees=45
  twistMotion=Free
 }
]
